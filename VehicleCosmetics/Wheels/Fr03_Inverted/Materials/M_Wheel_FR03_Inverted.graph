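# Unlit emissive material for the FR03 Inverted vehicle wheel, tinted by an Emissive Color parameter.
# Two panning samples of a shared geometric texture, scaled by the tire emissive mask, form an animated glow that blends toward white via rim mask blue, color and Tread Lights.
# A masked Fresnel rim is added on top.
# The result passes through shared vehicle damage and de-rez transition functions as material attributes.
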Material /VehicleCosmetics/Wheels/Fr03_Inverted/Materials/M_Wheel_FR03_Inverted
graph {
 properties {
  bAutomaticallySetUsageInEditor=false
  bUseMaterialAttributes=true
  bUsedWithMeshDeformer=true
  bUsedWithSkeletalMesh=true
  shadingModel=MSM_Unlit
  usedShadingModels=Unlit
 }
 nodes [
  n1 = TextureSampleParameter2D() {
   parameterName=Tire Emissive
   samplerSource=SSM_Wrap_WorldGroupSettings
   samplerType=SAMPLERTYPE_LinearColor
   texture=/VehicleCosmetics/Wheels/Fr03_Inverted/Textures/T_Wheel_FR_03_Inverted_Rim_M
  }
  n2 = ScalarParameter() {defaultValue=1, parameterName=TireEmissiveGlow}
  n3 = TextureCoordinate() {uTiling=0.33}
  n4 = Multiply(
   A=Multiply(A=@n1.R, B=@n2)
   B=TextureSample(UVs=Panner(Coordinate=@n3) {speedX=0.5}) {samplerSource=SSM_Wrap_WorldGroupSettings, texture=/VehicleCosmetics/Wheels/_SharedTextures/FX/Skin_GeometricBM_RGB}.R
  )
  n5 = Multiply(A=@n1.R, B=@n2)
  n6 = TextureSample(UVs=Panner() {speedX=0.4}) {samplerSource=SSM_Wrap_WorldGroupSettings, texture=&Skin_GeometricBM_RGB}
  n7 = Clamp(None=Add(A=Floor(None=@n4), B=Multiply(A=@n5, B=@n6.R))) {maxDefault=3}
  n8 = LinearInterpolate(A=@n7, B=@n7, Alpha=VertexColor().R)
  n9 = VectorParameter() {group=5_Special Effects, parameterName=Emissive Color, sortPriority=0, defaultValue=<rgba 0.1296 0.2054 1 1>}
  n10 = TextureSampleParameter2D() {
   parameterName=Mask
   samplerSource=SSM_Wrap_WorldGroupSettings
   samplerType=SAMPLERTYPE_LinearColor
   texture=/VehicleCosmetics/Wheels/Fr03_Inverted/Textures/T_Wheel_FR_03_Inverted_Rim_RGB
  }
  n11 = Multiply(A=@n10.B, B=@n9.RGB)
  n12 = ScalarParameter() {defaultValue=6, parameterName=Tread Lights}
  n13 = MaterialFunctionCall!(Normal, Exponent=ScalarParameter() {defaultValue=1.5, parameterName=Fresnel}) {materialFunction=/VehicleCosmetics/Wheels/Cola_Holographic/Materials/MF_Fresnel_Classic}
  n14 = TextureSampleParameter2D() {
   parameterName=Mask for Fresnel
   samplerSource=SSM_Wrap_WorldGroupSettings
   texture=/VehicleCosmetics/Wheels/Demoneer_Inverted/Textures/Wheel_demoneer_invert_M
  }
  n15 = Add(
   A=LinearInterpolate(A=Multiply(A=@n8, B=@n9.RGB), Alpha=Multiply(A=@n11, B=@n12))
   B=Multiply(A=@n9.RGB, B=Multiply(A=@n13.Result, B=@n14.RGB))
  )
  MaterialFunctionCall(In=MaterialFunctionCall(Material=SetMaterialAttributes(Emissive Color=@n15) {attributeSetTypes [#1]}) {materialFunction=&MF_BaseVehicle_Damage}.Result) {materialFunction=&MF_VehicleDeRez_Transition}
 ]
}
